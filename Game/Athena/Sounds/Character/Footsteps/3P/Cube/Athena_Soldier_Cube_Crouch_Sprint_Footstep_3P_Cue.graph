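# Third-person crouch-sprint footstep cue for Cube surfaces in Athena, routed to the enemy foley sound class.
# A quality-level node sends its first two levels to ten random run clips and its last two to a four-clip subset, each through a modulator.
# It reuses first-person Cubetown waves at half volume, with crouch-sprint attenuation and local and global footstep concurrency sets.
# The four-clip random node is excluded from branch culling.
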SoundCue /Game/Athena/Sounds/Character/Footsteps/3P/Cube/Athena_Soldier_Cube_Crouch_Sprint_Footstep_3P_Cue
root=QualityLevel_0
nodes [
 {id=Modulator_0, children [Random_1]}
 {id=Modulator_1, children [Random_2]}
 {id=QualityLevel_0, children [Modulator_1,Modulator_1,Modulator_0,Modulator_0]}
 {
  id=Random_1
  children [WavePlayer_28, WavePlayer_29, WavePlayer_30, WavePlayer_31]
  properties {bShouldExcludeFromBranchCulling=true, weights [1,1,1,1]}
 }
 {
  id=Random_2
  children [
   WavePlayer_3
   WavePlayer_4
   WavePlayer_5
   WavePlayer_7
   WavePlayer_10
   WavePlayer_21
   WavePlayer_23
   WavePlayer_24
   WavePlayer_25
   WavePlayer_26
  ]
  properties {weights [1,1,1,1,1,1,1,1,1,1]}
 }
 {id=WavePlayer_10, properties {soundWaveAssetPtr=/Game/Athena/Sounds/Character/Footsteps/1P/Cube/FS_Cubetown_Run_04}}
 {id=WavePlayer_21, properties {soundWaveAssetPtr=/Game/Athena/Sounds/Character/Footsteps/1P/Cube/FS_Cubetown_Run_05}}
 {id=WavePlayer_23, properties {soundWaveAssetPtr=/Game/Athena/Sounds/Character/Footsteps/1P/Cube/FS_Cubetown_Run_06}}
 {id=WavePlayer_24, properties {soundWaveAssetPtr=/Game/Athena/Sounds/Character/Footsteps/1P/Cube/FS_Cubetown_Run_07}}
 {id=WavePlayer_25, properties {soundWaveAssetPtr=/Game/Athena/Sounds/Character/Footsteps/1P/Cube/FS_Cubetown_Run_08}}
 {id=WavePlayer_26, properties {soundWaveAssetPtr=/Game/Athena/Sounds/Character/Footsteps/1P/Cube/FS_Cubetown_Run_09}}
 {id=WavePlayer_28, properties {soundWaveAssetPtr=/Game/Athena/Sounds/Character/Footsteps/1P/Cube/FS_Cubetown_Run_01}}
 {id=WavePlayer_29, properties {soundWaveAssetPtr=/Game/Athena/Sounds/Character/Footsteps/1P/Cube/FS_Cubetown_Run_02}}
 {id=WavePlayer_3, properties {soundWaveAssetPtr=/Game/Athena/Sounds/Character/Footsteps/1P/Cube/FS_Cubetown_Run_10}}
 {id=WavePlayer_30, properties {soundWaveAssetPtr=/Game/Athena/Sounds/Character/Footsteps/1P/Cube/FS_Cubetown_Run_03}}
 {id=WavePlayer_31, properties {soundWaveAssetPtr=&FS_Cubetown_Run_04}}
 {id=WavePlayer_4, properties {soundWaveAssetPtr=&FS_Cubetown_Run_01}}
 {id=WavePlayer_5, properties {soundWaveAssetPtr=&FS_Cubetown_Run_02}}
 {id=WavePlayer_7, properties {soundWaveAssetPtr=&FS_Cubetown_Run_03}}
]
properties {
 attenuationSettings=/Game/Sounds/Attenuation/Fort_Player_Footsteps_3P_CrouchSprint_Athena_Att
 maxDistance=2560
 soundClassObject=&Fort_Player_Foley_3P_Enemy_Athena
 volumeMultiplier=0.5
 concurrencyOverrides {bLimitToOwner=true, maxCount=1, resolutionRule=StopOldest}
 concurrencySet [&Athena_Footstep_Local_Conc, &Athena_Footstep_Global_Conc]
}
