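# Sound cue for third-person walking footsteps on the Cube (Cubetown) surface, using enemy player foley class and 3P walk attenuation.
# A quality-level node gives the first two levels a modulated random pick from ten walk waves, and the last two a random pick from four of them.
# The waves come from the 1P Cube folder.
# Concurrency sets are assigned alongside an override allowing one instance per owner, stopping the oldest.
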SoundCue /Game/Athena/Sounds/Character/Footsteps/3P/Cube/Athena_Soldier_Cube_Walk_Footstep_3P_Cue
root=QualityLevel_0
nodes [
 {id=Modulator_0, children [Random_2]}
 {id=Modulator_1, children [Random_1]}
 {id=QualityLevel_0, children [Modulator_0,Modulator_0,Modulator_1,Modulator_1]}
 {
  id=Random_1
  children [WavePlayer_25, WavePlayer_26, WavePlayer_27, WavePlayer_28]
  properties {bShouldExcludeFromBranchCulling=true, weights [1,1,1,1]}
 }
 {
  id=Random_2
  children [
   WavePlayer_6
   WavePlayer_7
   WavePlayer_8
   WavePlayer_9
   WavePlayer_10
   WavePlayer_11
   WavePlayer_20
   WavePlayer_21
   WavePlayer_23
   WavePlayer_24
  ]
  properties {weights [1,1,1,1,1,1,1,1,1,1]}
 }
 {id=WavePlayer_10, properties {soundWaveAssetPtr=/Game/Athena/Sounds/Character/Footsteps/1P/Cube/FS_Cubetown_Walk_05}}
 {id=WavePlayer_11, properties {soundWaveAssetPtr=/Game/Athena/Sounds/Character/Footsteps/1P/Cube/FS_Cubetown_Walk_06}}
 {id=WavePlayer_20, properties {soundWaveAssetPtr=/Game/Athena/Sounds/Character/Footsteps/1P/Cube/FS_Cubetown_Walk_07}}
 {id=WavePlayer_21, properties {soundWaveAssetPtr=/Game/Athena/Sounds/Character/Footsteps/1P/Cube/FS_Cubetown_Walk_08}}
 {id=WavePlayer_23, properties {soundWaveAssetPtr=/Game/Athena/Sounds/Character/Footsteps/1P/Cube/FS_Cubetown_Walk_09}}
 {id=WavePlayer_24, properties {soundWaveAssetPtr=/Game/Athena/Sounds/Character/Footsteps/1P/Cube/FS_Cubetown_Walk_10}}
 {id=WavePlayer_25, properties {soundWaveAssetPtr=/Game/Athena/Sounds/Character/Footsteps/1P/Cube/FS_Cubetown_Walk_01}}
 {id=WavePlayer_26, properties {soundWaveAssetPtr=/Game/Athena/Sounds/Character/Footsteps/1P/Cube/FS_Cubetown_Walk_02}}
 {id=WavePlayer_27, properties {soundWaveAssetPtr=/Game/Athena/Sounds/Character/Footsteps/1P/Cube/FS_Cubetown_Walk_03}}
 {id=WavePlayer_28, properties {soundWaveAssetPtr=/Game/Athena/Sounds/Character/Footsteps/1P/Cube/FS_Cubetown_Walk_04}}
 {id=WavePlayer_6, properties {soundWaveAssetPtr=&FS_Cubetown_Walk_01}}
 {id=WavePlayer_7, properties {soundWaveAssetPtr=&FS_Cubetown_Walk_02}}
 {id=WavePlayer_8, properties {soundWaveAssetPtr=&FS_Cubetown_Walk_03}}
 {id=WavePlayer_9, properties {soundWaveAssetPtr=&FS_Cubetown_Walk_04}}
]
properties {
 attenuationSettings=/Game/Sounds/Attenuation/Fort_Player_Footsteps_3P_Walk_Athena_Att
 maxDistance=3584
 soundClassObject=&Fort_Player_Foley_3P_Enemy_Athena
 volumeMultiplier=0.65
 concurrencyOverrides {bLimitToOwner=true, maxCount=1, resolutionRule=StopOldest}
 concurrencySet [&Athena_Footstep_Local_Conc, &Athena_Footstep_Global_Conc]
}
